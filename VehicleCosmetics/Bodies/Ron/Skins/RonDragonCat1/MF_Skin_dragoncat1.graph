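# Material function for the Ron DragonCat1 vehicle skin, setting dark animated lines over the base color and yellow, audio-reactive emissive glow.
# World-Z bands scroll over time, distorted by Perlin noise, to blend in grey.
# Emissive combines RGB-texture masks with three procedural wave-line layers built from noise, cosine time and ceil/floor distance.
# MF_AudioBeatSync spectrum becomes low, mid and high signals, shared alongside masks through named reroutes.
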MaterialFunction /VehicleCosmetics/Bodies/Ron/Skins/RonDragonCat1/MF_Skin_dragoncat1
graph {nodes [
  n1 = TextureSampleParameter2D() {parameterName=BaseColorMap, texture=&/T_RonDragoncat1_BC}
  n2 = ConstantBiasScale(None=NamedRerouteUsage()) {bias=0.05, scale=10}
  n3 = WorldPosition()
  n4 = Multiply(A=Time() {bOverride_Period=true, period=6000}) {constB=-1}
  n5 = Constant4Vector() {constant=<rgba 2 4 0.1 1>}
  n6 = MaterialFunctionCall!(Texture Coordinates, Tiling X=@n5.R, Tiling Y=@n5.R, PanSpeedX=@n5.R, PanSpeedY=@n5.R, Angle, RotateSpeed, Balance, Chaos) {materialFunction=/VehicleCosmetics/Materials/MaterialFunctions/MF_Noisemaker_Perlin}
  n7 = Multiply(A=Add(A=Add(A=@n3.Z, B=@n4), B=Multiply(A=@n6.Perlin2) {constB=0.5})) {constB=0.34}
  n8 = Subtract(A=Subtract(A=@n7, B=Floor(None=@n7))) {constB=0.9}
  n9 = Subtract(A=Saturate(None=Multiply(A=@n2, B=@n8)), B=NamedRerouteUsage())
  n10 = Constant3Vector() {constant=<rgba 0.8355 0.8122 0 1>}
  n11 = TextureSample() {texture=&/T_RonDragoncat1_RGB}
  n12 = NamedRerouteUsage()
  n13 = Abs(None=@n12)
  n14 = Saturate(None=Multiply(A=Multiply(A=@n13) {constB=5}, B=@n11.B) {constB=5})
  n15 = Subtract(A=Multiply(A=NamedRerouteDeclaration(None=@n11.R) {name=RGB WavesMask, nodeColor=<rgba 0 0 1 1>}) {constB=0.5}, B=NamedRerouteUsage())
  n16 = Multiply(A=Multiply(A=Saturate(None=@n15), B=NamedRerouteUsage())) {constB=0.25}
  n17 = Saturate(None=ConstantBiasScale(None=NamedRerouteUsage()) {bias=-0.065, scale=1})
  n18 = NamedRerouteUsage()
  n19 = Constant4Vector() {constant=<rgba 10 2 0 1>}
  n20 = MaterialFunctionCall(Tiling X=@n19.R, Tiling Y=@n19.R, PanSpeedX=@n19.R, PanSpeedY=@n19.R) {materialFunction=&MF_Noisemaker_Perlin}
  n21 = Abs(None=Multiply(A=@n18) {constB=2})
  n22 = Add(
   A=Add(A=ConstantBiasScale(None=@n18) {bias=-0.5, scale=1}) {constB=2}
   B=LinearInterpolate(A=NamedRerouteUsage(), B=ConstantBiasScale(None=@n20.Perlin2) {bias=-0.5, scale=0.2}, Alpha=@n21)
  ) {constB=3}
  n23 = NamedRerouteUsage()
  n24 = Cosine(None=Time() {bOverride_Period=true, period=6000}) {period=6}
  n25 = Add(A=Multiply(A=Multiply(A=@n22, B=@n23)) {constB=1.75}, B=Add(A=Multiply(A=@n24) {constB=0.4}) {constB=0.25})
  n26 = Ceil(None=@n25)
  n27 = Floor(None=@n25)
  n28 = Max(
   A=ConstantBiasScale(None=Subtract(A=@n26, B=@n25)) {bias=-0.7, scale=2}
   B=ConstantBiasScale(None=Subtract(A=@n25, B=@n27)) {bias=-0.7, scale=2}
  )
  n29 = Multiply(A=Multiply(A=Multiply(A=@n28) {constB=1.5}, B=@n21) {constB=1.5}) {constB=0.25}
  n30 = ConstantBiasScale(None=@n28) {bias=-0.45, scale=9}
  n31 = NamedRerouteUsage()
  n32 = Constant4Vector() {constant=<rgba 15 0.5 2 4>}
  n33 = MaterialFunctionCall(Tiling X=@n32.R, Tiling Y=@n32.R, PanSpeedX=Constant(), PanSpeedY=@n32.R, Balance=@n32.R) {materialFunction=&MF_Noisemaker_Perlin}
  n34 = Abs(None=Multiply(A=@n31) {constB=2})
  n35 = Add(
   A=Add(A=ConstantBiasScale(None=@n31) {bias=-0.5, scale=1}) {constB=1.5}
   B=LinearInterpolate(A=NamedRerouteUsage(), B=ConstantBiasScale(None=@n33.Perlin2) {bias=-0.5, scale=0.6}, Alpha=@n34)
  ) {constB=3}
  n36 = NamedRerouteUsage()
  n37 = Cosine(None=Time() {bOverride_Period=true, period=6000}) {period=4}
  n38 = Add(A=Multiply(A=Multiply(A=@n35, B=@n36)) {constB=1.5}, B=Add(A=Multiply(A=@n37) {constB=0.2}) {constB=0.3})
  n39 = Ceil(None=@n38)
  n40 = Floor(None=@n38)
  n41 = Max(
   A=ConstantBiasScale(None=Subtract(A=@n39, B=@n38)) {bias=-0.7, scale=2}
   B=ConstantBiasScale(None=Subtract(A=@n38, B=@n40)) {bias=-0.7, scale=2}
  )
  n42 = Multiply(A=@n41) {constB=1.5}
  n43 = Max(A=Multiply(A=Multiply(A=@n42, B=@n34) {constB=1.5}) {constB=0.5}, B=ConstantBiasScale(None=@n41) {bias=-0.45, scale=7})
  n44 = ConstantBiasScale(None=NamedRerouteUsage()) {bias=-0.5, scale=2}
  n45 = Constant4Vector() {constant=<rgba 12 1 2 3>}
  n46 = MaterialFunctionCall(Tiling X=@n45.R, Tiling Y=@n45.R, PanSpeedX=Constant(), PanSpeedY=@n45.R, Balance=@n45.R) {materialFunction=&MF_Noisemaker_Perlin}
  n47 = OneMinus(None=Abs(None=@n44))
  n48 = Add(
   A=Add(A=@n44) {constB=3}
   B=LinearInterpolate(A=NamedRerouteUsage(), B=ConstantBiasScale(None=@n46.Perlin2) {bias=-0.5, scale=0.6}, Alpha=@n47)
  ) {constB=3}
  n49 = NamedRerouteUsage()
  n50 = Cosine(None=Time() {bOverride_Period=true, period=6000}) {period=3}
  n51 = Add(A=Multiply(A=Multiply(A=@n48, B=@n49)), B=Add(A=Multiply(A=@n50) {constB=0.1}) {constB=0.4})
  n52 = Ceil(None=@n51)
  n53 = Floor(None=@n51)
  n54 = Max(
   A=ConstantBiasScale(None=Subtract(A=@n52, B=@n51)) {bias=-0.7, scale=2}
   B=ConstantBiasScale(None=Subtract(A=@n51, B=@n53)) {bias=-0.7, scale=2}
  )
  n55 = Multiply(A=Multiply(A=@n54) {constB=1.5}, B=Saturate(None=@n47)) {constB=1.5}
  n56 = Max(A=Saturate(None=Multiply(A=@n55) {constB=0.25}), B=ConstantBiasScale(None=@n54) {bias=-0.45, scale=6})
  n57 = Max(A=Saturate(None=Max(A=@n29, B=@n30)), B=Max(A=Saturate(None=@n43), B=Saturate(None=@n56)))
  n58 = Max(A=Max(A=Multiply(A=@n11.G, B=@n12) {constB=5}, B=Max(A=@n14, B=@n16)), B=Subtract(A=Multiply(A=@n17, B=@n57), B=NamedRerouteUsage()))
  n59 = SetMaterialAttributes!(
   MaterialAttributes
   BaseColor=LinearInterpolate(A=@n1.RGB, B=Constant3Vector() {constant=<rgba 0.1361 0.1361 0.1361 1>}, Alpha=Saturate(None=@n9))
   Metallic=Constant() {r=0.2}
   Roughness=Constant() {r=0.2}
   EmissiveColor=Multiply(A=Multiply(A=@n10, B=@n58), B=ScalarParameter() {defaultValue=2, parameterName=Emissive Intensity})
   Normal
   AmbientOcclusion
  ) {attributeSetTypes [
    #1
    #2
    #3
    #4
    #5
    #6
   ]}
  FunctionOutput(None=@n59)
  NamedRerouteDeclaration(None=@n1.A) {name=ColorMapAlpha, nodeColor=<rgba 0.9219 1 0 1>}
  n60 = MaterialFunctionCall!(FakeBeatIntensity, BeatSyncIntensity=ScalarParameter() {defaultValue=1, parameterName=BeatSyncIntensity}) {materialFunction=/VehicleCosmetics/SharedMaterials/MaterialFunctions/MF_AudioBeatSync}
  n61 = ComponentMask(None=@n60.Spectrum) {b=true}
  n62 = NamedRerouteDeclaration(None=Max(A=@n61, B=ComponentMask(None=@n60.Spectrum) {a=true})) {name=BeatSync Hi, nodeColor=<rgba 0.724 0.0103 0 1>}
  n63 = NamedRerouteDeclaration(None=Add(A=ComponentMask(None=@n60.Spectrum) {r=true}, B=ComponentMask(None=@n60.Spectrum) {g=true})) {name=BeatSync Low, nodeColor=<rgba 0 0.2957 0.0475 1>}
  NamedRerouteDeclaration(None=Max(A=Add(A=@n62, B=@n63), B=@n61)) {name=BeatSync Mid, nodeColor=<rgba 1 0.9198 0 1>}
  NamedRerouteDeclaration(None=Multiply(A=@n6.Perlin2) {constB=0.2}) {name=Grey Line Noise, nodeColor=<rgba 0 1 0.1094 1>}
 ]}
